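# EQS query that picks a flee destination for ProtoSwarm AI.
# A donut of navmesh-projected points around the querier is filtered to the data-bound leash radius, existing paths, and spots clear of nearby AI pawns and their move destinations.
# Scoring favors distance from self and from nearby friendly AI and players, with leash-center proximity, a dot test against U-turns and slight randomness.
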EnvQuery /ProtoSwarm/EQS/EQS_ProtoSwarm_Flee
options [{
  generator {
   name=EnvQueryGenerator_Donut_2
   class=EnvQueryGenerator_Donut
   properties {
    bUseSpiralPattern=false
    center=&EnvQueryContext_Querier
    innerRadius=10
    numberOfRings=3
    outerRadius=1000
    pointsPerRing=6
    arcDirection {lineFrom=null}
    projectionData {
     extentX=100
     navigationFilter=&NavFilter_Phoebe_NoSmashableNoWater_C
     postProjectionVerticalOffset=30
    }
   }
  }
  tests [
   EnvQueryTest_Distance_13 {
    class=EnvQueryTest_Distance
    properties {
     clampMinType=SpecifiedValue
     distanceTo=/Script/FortniteAI.FortQueryContext_LeashCenter
     filterType=Maximum
     floatValueMin=0
     scoreClampMax=2000
     scoreClampMin=2000
     scoringEquation=InverseLinear
     scoringFactor=0.75
     testComment=Inside leash
     testMode=Distance2D
     testPurpose=FilterAndScore
     floatValueMax {
      dataBinding=&self:EnvQueryTest_Distance_13.FortAIDataProvider_LeashComponent_0
      dataField=LeashRadius
      defaultValue=0
     }
    }
   }
   FortQueryTest_PathfindingBatch_3 {
    class=FortQueryTest_PathfindingBatch
    properties {
     bDefineReferenceValue=false
     boolValue=true
     clampMaxType=None
     clampMinType=None
     context=&EnvQueryContext_Querier
     filterClass=/Game/Athena/AI/NavFilters/NavFilter_NPC_ExcludeWater.NavFilter_NPC_ExcludeWater_C
     filterType=Match
     floatValueMax=1000
     floatValueMin=1000
     multipleContextFilterOp=AllPass
     multipleContextScoreOp=AverageScore
     navDataOverrideContext=null
     normalizationType=Absolute
     pathFromContext=true
     referenceValue=0
     scanRangeMultiplier=1.5
     scoreClampMax=0
     scoreClampMin=0
     scoringEquation=Constant
     scoringFactor=1
     skipUnreachable=true
     testComment=Throw out things you cannot path to.
     testMode=PathExist
     testPurpose=Filter
     navFilterTag {tagName=None}
    }
   }
   EnvQueryTest_Distance_14 {
    class=EnvQueryTest_Distance
    properties {
     distanceTo=&EnvQueryContext_Querier
     filterType=Range
     floatValueMin=0
     scoringEquation=Linear
     scoringFactor=1
     testComment=Pick point farthest from Self
     testMode=Distance2D
     testPurpose=Score
    }
   }
   EnvQueryTest_Dot_3 {
    class=EnvQueryTest_Dot
    properties {clampMaxType=SpecifiedValue, clampMinType=None, scoreClampMax=0.5, scoreClampMin=0, scoringFactor=1, testComment=Avoid 180 degrees turns}
   }
   FortQueryTest_Random_3 {class=FortQueryTest_Random, properties {floatValueMax=0.1, scoringFactor=1, testComment=Random Score}}
   EnvQueryTest_Distance_15 {
    class=EnvQueryTest_Distance
    properties {
     distanceTo=&FortQueryContext_NearbyAIPawns
     filterType=Minimum
     floatValueMin=80
     scoringEquation=InverseLinear
     scoringFactor=3
     testComment=Try not to overlap AI Pawns
     testMode=Distance3D
     testPurpose=Filter
    }
   }
   EnvQueryTest_Distance_16 {
    class=EnvQueryTest_Distance
    properties {
     distanceTo=&FortQueryContext_NearbyAIPawnsMoveDestinations
     filterType=Minimum
     floatValueMin=80
     scoringEquation=Linear
     scoringFactor=1
     testComment=Not near other pawn move destinations
     testMode=Distance3D
     testPurpose=Filter
    }
   }
   EnvQueryTest_Distance_17 {
    class=EnvQueryTest_Distance
    properties {
     distanceTo=/Script/FortniteAIServer.FortQueryContext_NearbyFriendlyAIPawnsAndPlayerPawns
     filterType=Range
     floatValueMin=0
     scoringEquation=Linear
     scoringFactor=1
     testComment=""
     testMode=Distance2D
     testPurpose=Score
    }
   }
  ]
 }]
