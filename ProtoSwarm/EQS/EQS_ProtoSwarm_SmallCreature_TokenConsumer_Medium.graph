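# EQS query for ProtoSwarm small creatures that, as token consumers, pick a medium-range token position around their goal.
# A token provider generator gathers positions from the goal context, using a tag query for the medium-range AI token.
# A single 3D distance test against the querier scores them linearly with a negative factor, favoring the closest position without filtering any out.
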EnvQuery /ProtoSwarm/EQS/EQS_ProtoSwarm_SmallCreature_TokenConsumer_Medium
options [{
  generator {
   name=FortQueryGenerator_TokenProviderPositions_0
   class=FortQueryGenerator_TokenProviderPositions
   properties {
    bAutoSortTests=true
    bCanRunAsync=false
    envQueryResultNormalizationOption=Default
    optionName=""
    tokenProviderContext=&FortQueryContext_Goal
    gameplayTagQuery {
     autoDescription=" ANY( Athena.AI.Token.Range.Medium )"
     tokenStreamVersion=0
     userDescription=""
     queryTokenStream [0,1,1,1,0]
     tagDictionary [{tagName=Athena.AI.Token.Range.Medium}]
    }
   }
  }
  tests [EnvQueryTest_Distance_1 {
    class=EnvQueryTest_Distance
    properties {
     distanceTo=&EnvQueryContext_Querier
     filterType=Minimum
     floatValueMax=500
     floatValueMin=0
     scoringEquation=Linear
     scoringFactor=-1
     testComment=""
     testMode=Distance3D
     testPurpose=Score
    }
   }]
 }]
