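# ProtoSwarm EQS query that picks a sideways movement point for zigzagging relative to the AI's goal.
# It generates a few points on a goal-facing half circle, projected onto navmesh, and uses pathfinding with a no-smashable, no-water filter to filter and score them.
# Scoring favors points perpendicular to the goal-querier line, far from the querier and near the goal.
# The leash radius feeds the goal-distance test's maximum, but that test only scores.
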EnvQuery /ProtoSwarm/EQS/EQS_ProtoSwarm_ZigZag
options [{
  generator {
   name=EnvQueryGenerator_OnCircle_0
   class=EnvQueryGenerator_OnCircle
   properties {
    arcAngle=180
    bDefineArc=true
    circleRadius=1200
    numberOfPoints=6
    arcDirection {lineTo=&FortQueryContext_Goal}
    projectionData {
     extentX=100
     navigationFilter=&NavFilter_Phoebe_NoSmashableNoWater_C
     postProjectionVerticalOffset=0
     projectDown=1024
    }
    traceData {navigationFilter=null}
   }
  }
  tests [
   EnvQueryTest_Dot_1 {
    class=EnvQueryTest_Dot
    properties {
     bAbsoluteValue=true
     clampMaxType=SpecifiedValue
     clampMinType=SpecifiedValue
     floatValueMax=100
     floatValueMin=70
     scoreClampMax=1
     scoreClampMin=0
     scoringFactor=-1
     testComment=Favor points perpendicular
     lineA {lineFrom=&FortQueryContext_Goal, lineTo=&EnvQueryContext_Querier}
    }
   }
   EnvQueryTest_Pathfinding_1 {
    class=EnvQueryTest_Pathfinding
    properties {
     filterClass=&NavFilter_Phoebe_NoSmashableNoWater_C
     testPurpose=FilterAndScore
    }
   }
   EnvQueryTest_Distance_1 {
    class=EnvQueryTest_Distance
    properties {
     distanceTo=&EnvQueryContext_Querier
     filterType=Range
     floatValueMin=0
     scoringEquation=Linear
     scoringFactor=1
     testComment=Pick point farthest from Self
     testMode=Distance2D
     testPurpose=Score
    }
   }
   EnvQueryTest_Distance_2 {
    class=EnvQueryTest_Distance
    properties {
     distanceTo=&FortQueryContext_Goal
     filterType=Maximum
     floatValueMin=0
     scoreClampMax=2000
     scoreClampMin=2000
     scoringEquation=InverseLinear
     scoringFactor=1
     testComment=""
     testMode=Distance2D
     testPurpose=Score
     floatValueMax {
      dataBinding=&self:EnvQueryTest_Distance_2.FortAIDataProvider_LeashComponent_0
      dataField=LeashRadius
      defaultValue=0
     }
    }
   }
  ]
 }]
