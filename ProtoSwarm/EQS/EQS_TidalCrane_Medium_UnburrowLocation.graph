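# EQS query that picks where a medium TidalCrane AI unburrows, favoring points far from itself and within its leash.
# First option filters custom unburrow points by range, AI-pawn spacing and a trace from the goal.
# Fallback samples a navmesh-projected donut around the leash center, adding building, water and pathfinding tests, with traces only scoring.
# Last-resort option repeats that donut without the tests against other pawns and their move destinations.
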EnvQuery /ProtoSwarm/EQS/EQS_TidalCrane_Medium_UnburrowLocation
options [
 {
  generator {
   name=FortQueryGenerator_TidalCrane_Medium_UnburrowPoints_C_0
   class=FortQueryGenerator_TidalCrane_Medium_UnburrowPoints_C
   properties {
    bAutoSortTests=true
    bCanRunAsync=false
    context=&/FortQueryContext_TidalCrane_Medium_UnborrowPoints.FortQueryContext_TidalCrane_Medium_UnborrowPoints_C
    envQueryResultNormalizationOption=Default
    generatedItemType=/Script/AIModule.EnvQueryItemType_Point
    generatorsActionDescription=""
    optionName=""
   }
  }
  tests [
   EnvQueryTest_Distance_5 {
    class=EnvQueryTest_Distance
    properties {
     clampMinType=SpecifiedValue
     distanceTo=/Script/FortniteAI.FortQueryContext_LeashCenter
     filterType=Maximum
     floatValueMin=0
     scoreClampMax=2000
     scoreClampMin=2000
     scoringEquation=InverseLinear
     scoringFactor=0.75
     testComment=Inside leash
     testMode=Distance2D
     testPurpose=FilterAndScore
     floatValueMax {
      dataBinding=&self:EnvQueryTest_Distance_5.FortAIDataProvider_LeashComponent_0
      dataField=LeashRadius
      defaultValue=0
     }
    }
   }
   EnvQueryTest_Distance_2 {
    class=EnvQueryTest_Distance
    properties {
     distanceTo=&EnvQueryContext_Querier
     filterType=Maximum
     floatValueMin=0
     scoringEquation=Linear
     scoringFactor=1
     testComment=""
     testMode=Distance3D
     testPurpose=Filter
     floatValueMax {
      dataBinding=&self:EnvQueryTest_Distance_2.FortAIDataProvider_ScalableFloat_0
      dataField=FloatValue
      defaultValue=6000
     }
    }
   }
   EnvQueryTest_Distance_6 {
    class=EnvQueryTest_Distance
    properties {
     distanceTo=&EnvQueryContext_Querier
     filterType=Range
     floatValueMin=0
     scoringEquation=Linear
     scoringFactor=1
     testComment=Pick point farthest from Self
     testMode=Distance2D
     testPurpose=Score
    }
   }
   EnvQueryTest_Distance_7 {
    class=EnvQueryTest_Distance
    properties {
     distanceTo=&FortQueryContext_NearbyAIPawns
     filterType=Minimum
     floatValueMin=300
     scoringEquation=InverseLinear
     scoringFactor=3
     testComment=Try not to overlap AI Pawns
     testMode=Distance3D
     testPurpose=Filter
    }
   }
   EnvQueryTest_Distance_8 {
    class=EnvQueryTest_Distance
    properties {
     distanceTo=&FortQueryContext_NearbyAIPawnsMoveDestinations
     filterType=Minimum
     floatValueMin=300
     scoringEquation=Linear
     scoringFactor=1
     testComment=Not near other pawn move destinations
     testMode=Distance3D
     testPurpose=Filter
    }
   }
   EnvQueryTest_Distance_9 {
    class=EnvQueryTest_Distance
    properties {
     distanceTo=/Script/FortniteAIServer.FortQueryContext_NearbyFriendlyAIPawnsAndPlayerPawns
     filterType=Range
     floatValueMin=0
     scoringEquation=Linear
     scoringFactor=1
     testComment=""
     testMode=Distance2D
     testPurpose=Score
    }
   }
   EnvQueryTest_Trace_0 {
    class=EnvQueryTest_Trace
    properties {
     context=&FortQueryContext_Goal
     itemHeightOffset=50
     multipleContextScoreOp=Multiply
     scoringFactor=2
     testComment=""
     testPurpose=Filter
     traceFromContext=true
    }
   }
   FortQueryTest_Random_1 {class=FortQueryTest_Random, properties {floatValueMax=0.1, scoringFactor=1, testComment=Random Score}}
  ]
 }
 {
  generator {
   name=EnvQueryGenerator_Donut_2
   class=EnvQueryGenerator_Donut
   properties {
    bUseSpiralPattern=true
    center=&FortQueryContext_LeashCenter
    innerRadius=2000
    numberOfRings=4
    outerRadius=6000
    pointsPerRing=12
    arcDirection {lineFrom=null}
    projectionData {
     extentX=100
     navigationFilter=&NavFilter_Phoebe_NoSmashableNoWater_C
     postProjectionVerticalOffset=30
     projectUp=4000
    }
   }
  }
  tests [
   EnvQueryTest_Distance_17 {
    class=EnvQueryTest_Distance
    properties {
     clampMinType=SpecifiedValue
     distanceTo=&FortQueryContext_LeashCenter
     filterType=Maximum
     floatValueMin=0
     scoreClampMax=2000
     scoreClampMin=2000
     scoringEquation=InverseLinear
     scoringFactor=0.75
     testComment=Inside leash
     testMode=Distance2D
     testPurpose=FilterAndScore
     floatValueMax {
      dataBinding=&self:EnvQueryTest_Distance_17.FortAIDataProvider_LeashComponent_0
      dataField=LeashRadius
      defaultValue=0
     }
    }
   }
   EnvQueryTest_Distance_3 {
    class=EnvQueryTest_Distance
    properties {
     distanceTo=&EnvQueryContext_Querier
     filterType=Maximum
     floatValueMin=0
     scoringEquation=Linear
     scoringFactor=1
     testComment=""
     testMode=Distance3D
     testPurpose=Filter
     floatValueMax {
      dataBinding=&self:EnvQueryTest_Distance_3.FortAIDataProvider_ScalableFloat_0
      dataField=FloatValue
      defaultValue=6000
     }
    }
   }
   EnvQueryTest_Distance_18 {
    class=EnvQueryTest_Distance
    properties {
     distanceTo=&EnvQueryContext_Querier
     filterType=Range
     floatValueMin=0
     scoringEquation=Linear
     scoringFactor=2.5
     testComment=Pick point farthest from Self
     testMode=Distance2D
     testPurpose=Score
    }
   }
   FortQueryTest_Random_4 {class=FortQueryTest_Random, properties {floatValueMax=0.1, scoringFactor=1, testComment=Random Score}}
   EnvQueryTest_Distance_19 {
    class=EnvQueryTest_Distance
    properties {
     distanceTo=&FortQueryContext_NearbyAIPawns
     filterType=Minimum
     floatValueMin=300
     scoringEquation=InverseLinear
     scoringFactor=3
     testComment=Try not to overlap AI Pawns
     testMode=Distance3D
     testPurpose=Filter
    }
   }
   EnvQueryTest_Distance_20 {
    class=EnvQueryTest_Distance
    properties {
     distanceTo=&FortQueryContext_NearbyAIPawnsMoveDestinations
     filterType=Minimum
     floatValueMin=300
     scoringEquation=Linear
     scoringFactor=1
     testComment=Not near other pawn move destinations
     testMode=Distance3D
     testPurpose=Filter
    }
   }
   EnvQueryTest_Distance_21 {
    class=EnvQueryTest_Distance
    properties {
     distanceTo=&FortQueryContext_NearbyFriendlyAIPawnsAndPlayerPawns
     filterType=Range
     floatValueMin=0
     scoringEquation=Linear
     scoringFactor=1
     testComment=""
     testMode=Distance2D
     testPurpose=Score
    }
   }
   FortQueryTest_HasNearbyBuildings_3 {class=FortQueryTest_HasNearbyBuildings, properties {extentXY=1, testPurpose=Filter}}
   EnvQueryTest_Trace_2 {
    class=EnvQueryTest_Trace
    properties {
     context=&FortQueryContext_Goal
     contextHeightOffset=20
     itemHeightOffset=10
     multipleContextScoreOp=Multiply
     scoringFactor=1.5
     testComment=""
     testPurpose=Score
     traceFromContext=true
    }
   }
   EnvQueryTest_Trace_3 {
    class=EnvQueryTest_Trace
    properties {contextHeightOffset=20, multipleContextScoreOp=Multiply, scoringFactor=1.5, testComment="", testPurpose=Score, traceFromContext=true}
   }
   EnvQueryTest_Overlap_2 {
    class=EnvQueryTest_Overlap
    properties {testComment=Reject Water, overlapData {extentX=300, extentY=300, extentZ=300, overlapChannel=ECC_GameTraceChannel15, shapeOffset=<0 0 200>}}
   }
   EnvQueryTest_Pathfinding_0 {
    class=EnvQueryTest_Pathfinding
    properties {
     filterClass=&NavFilter_Phoebe_NoSmashableNoWater_C
     testPurpose=FilterAndScore
    }
   }
  ]
 }
 {
  generator {
   name=EnvQueryGenerator_Donut_0
   class=EnvQueryGenerator_Donut
   properties {
    bUseSpiralPattern=true
    center=&FortQueryContext_LeashCenter
    innerRadius=2000
    numberOfRings=4
    outerRadius=6000
    pointsPerRing=12
    arcDirection {lineFrom=null}
    projectionData {
     extentX=100
     navigationFilter=&NavFilter_Phoebe_NoSmashableNoWater_C
     postProjectionVerticalOffset=30
     projectUp=4000
    }
   }
  }
  tests [
   EnvQueryTest_Distance_0 {
    class=EnvQueryTest_Distance
    properties {
     clampMinType=SpecifiedValue
     distanceTo=&FortQueryContext_LeashCenter
     filterType=Maximum
     floatValueMin=0
     scoreClampMax=2000
     scoreClampMin=2000
     scoringEquation=InverseLinear
     scoringFactor=0.75
     testComment=Inside leash
     testMode=Distance2D
     testPurpose=FilterAndScore
     floatValueMax {
      dataBinding=&self:EnvQueryTest_Distance_0.FortAIDataProvider_LeashComponent_0
      dataField=LeashRadius
      defaultValue=0
     }
    }
   }
   EnvQueryTest_Distance_4 {
    class=EnvQueryTest_Distance
    properties {
     distanceTo=&EnvQueryContext_Querier
     filterType=Maximum
     floatValueMin=0
     scoringEquation=Linear
     scoringFactor=1
     testComment=""
     testMode=Distance3D
     testPurpose=Filter
     floatValueMax {
      dataBinding=&self:EnvQueryTest_Distance_4.FortAIDataProvider_ScalableFloat_0
      dataField=FloatValue
      defaultValue=6000
     }
    }
   }
   EnvQueryTest_Distance_1 {
    class=EnvQueryTest_Distance
    properties {
     distanceTo=&EnvQueryContext_Querier
     filterType=Range
     floatValueMin=0
     scoringEquation=Linear
     scoringFactor=2.5
     testComment=Pick point farthest from Self
     testMode=Distance2D
     testPurpose=Score
    }
   }
   FortQueryTest_Random_0 {class=FortQueryTest_Random, properties {floatValueMax=0.1, scoringFactor=1, testComment=Random Score}}
   FortQueryTest_HasNearbyBuildings_0 {class=FortQueryTest_HasNearbyBuildings, properties {extentXY=1, testPurpose=Filter}}
   EnvQueryTest_Trace_5 {
    class=EnvQueryTest_Trace
    properties {
     context=&FortQueryContext_Goal
     contextHeightOffset=20
     itemHeightOffset=10
     multipleContextScoreOp=Multiply
     scoringFactor=1.5
     testComment=""
     testPurpose=Score
     traceFromContext=true
    }
   }
   EnvQueryTest_Trace_4 {
    class=EnvQueryTest_Trace
    properties {contextHeightOffset=20, multipleContextScoreOp=Multiply, scoringFactor=1.5, testComment="", testPurpose=Score, traceFromContext=true}
   }
   EnvQueryTest_Overlap_1 {
    class=EnvQueryTest_Overlap
    properties {testComment=Reject Water, overlapData {extentX=300, extentY=300, extentZ=300, overlapChannel=ECC_GameTraceChannel15, shapeOffset=<0 0 200>}}
   }
   EnvQueryTest_Pathfinding_2 {
    class=EnvQueryTest_Pathfinding
    properties {
     filterClass=&NavFilter_Phoebe_NoSmashableNoWater_C
     testPurpose=FilterAndScore
    }
   }
  ]
 }
]
