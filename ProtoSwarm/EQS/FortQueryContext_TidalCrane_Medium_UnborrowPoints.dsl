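; EQS query context that supplies candidate unburrow points for the medium TidalCrane in ProtoSwarm.
; ProvideLocationsSet gathers every BP_TidalCrane_Medium_UnburrowLocation actor in the world and keeps those within MaxDistance of the querier, measured in 2D so height is ignored.
; It adds their locations uniquely to PossibleLocations and returns that set to the query.
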(fn ProvideLocationsSet
  (params (QuerierObject :type Object%20Reference) (QuerierActor :type Actor%20Object%20Reference))
  (bind _outactors (Actor|GetAllActorsOfClass "/ProtoSwarm/Blueprints/Medium/BP_TidalCrane_Medium_UnburrowLocation.BP_TidalCrane_Medium_UnburrowLocation_C"))
  (for _array_element _outactors
    (if (<= (Math|Vector|Distance2D(Vector) (Transformation|GetActorLocation _array_element) (Transformation|GetActorLocation QuerierActor)) (LocalVariables|GetMaxDistance))
      (Utilities|Array|AddUnique (LocalVariables|GetPossibleLocations) (Transformation|GetActorLocation _array_element))))
  (return (LocalVariables|GetPossibleLocations)))
